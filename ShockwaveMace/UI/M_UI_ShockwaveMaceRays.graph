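# Translucent UI material that draws two layers of radial rays for the Shockwave Mace UI.
# Two MF_UI_RadialSegments calls, each with its own rotation, segment count and spacing parameters, produce primary and secondary segment gradients.
# Named reroutes carry these and a smooth radial gradient mask into a blend from Edge Color to Color and an opacity chain with per-layer and global controls.
# Rotation is parameter-driven, with no time node in the graph.
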Material /ShockwaveMace/UI/M_UI_ShockwaveMaceRays
graph {
 properties {blendMode=BLEND_Translucent, materialDomain=MD_UI, pixelDepthOffsetMode=PDOM_Legacy, usedShadingModels=Default Lit}
 nodes [
  n1 = Add(A=NamedRerouteUsage(), B=NamedRerouteUsage())
  LinearInterpolate(
   A=VectorParameter() {parameterName=Edge Color, defaultValue=<rgba 0.826 0.8968 0.9167 1>}.RGB
   B=VectorParameter() {parameterName=Color, defaultValue=<rgba 0.2499 1 0.9901 1>}.RGB
   Alpha=Multiply(A=Saturate(None=@n1), B=NamedRerouteUsage())
  )
  n2 = Add(
   A=Multiply(A=NamedRerouteUsage(), B=ScalarParameter() {defaultValue=1, parameterName=Primary Opacity, sliderMax=1})
   B=Multiply(A=NamedRerouteUsage(), B=ScalarParameter() {defaultValue=0.5, parameterName=Secondary Opacity, sliderMax=1})
  )
  Multiply(A=Multiply(A=Saturate(None=@n2), B=NamedRerouteUsage()), B=ScalarParameter() {defaultValue=1, parameterName=Global Opacity, sliderMax=1})
  n3 = MaterialFunctionCall() {materialFunction=&GetUserInterfaceUV}
  n4 = ScalarParameter() {defaultValue=0.25, parameterName=Secondary Rotation}
  NamedRerouteDeclaration(None=MaterialFunctionCall!(
    UV=MaterialFunctionCall(UVs=@n3.Normalized UV, "Rotation Angle (0-1)"=@n4) {materialFunction=&CustomRotator}.Rotated Values
    Segment Count=ScalarParameter() {defaultValue=3, parameterName=Secondary Segments Count}
    Segment Spacing=ScalarParameter() {defaultValue=0.8, parameterName=Secondary Spacing}
    Fixed Spacing
   ) {materialFunction=/FortUILibrary/MaterialFunctions/Basics/MF_UI_RadialSegments}.Segmented Triangles Gradient) {name=Secondary Segments, nodeColor=<rgba 0 0 0 1>}
  n5 = ScalarParameter() {parameterName=Primary Rotation}
  NamedRerouteDeclaration(None=MaterialFunctionCall(
    UV=MaterialFunctionCall(UVs=@n3.Normalized UV, "Rotation Angle (0-1)"=@n5) {materialFunction=&CustomRotator}.Rotated Values
    Segment Count=ScalarParameter() {defaultValue=5, parameterName=Primary Segments Count}
    Segment Spacing=ScalarParameter() {defaultValue=0.8, parameterName=Primary Spacing}
   ) {materialFunction=&MF_UI_RadialSegments}.Segmented Triangles Gradient) {name=Primary Segments, nodeColor=<rgba 0 0 0 1>}
  NamedRerouteDeclaration(None=MaterialFunctionCall() {materialFunction=&MF_UI_RadialGradient}.Smooth) {name=Radial Mask, nodeColor=<rgba 0 0 0 1>}
  ScalarParameter() {defaultValue=0.04, parameterName=Spacing Falloff}
 ]
}
