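# Translucent UI material that draws the Shockwave Mace reticle procedurally from a segmented ring, mirrored triangles and a targeting circle.
# All shapes build on normalized UI UVs with a global rotation, with the triangles using 2x2 symmetry and a polygon SDF function.
# SDFs are feathered by SmoothStep, and radial segments are subtracted to gap the ring.
# Summed masks blend Outer Color to Color and, with per-part and global opacity, form the alpha.
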Material /ShockwaveMace/UI/M_UI_ShockwaveMaceReticle
graph {
 properties {blendMode=BLEND_Translucent, materialDomain=MD_UI, pixelDepthOffsetMode=PDOM_Legacy, usedShadingModels=Default Lit}
 nodes [
  Constant()
  n1 = NamedRerouteUsage()
  n2 = NamedRerouteUsage()
  n3 = NamedRerouteDeclaration(None=Saturate(None=Add(A=@n1, B=@n2))) {name=Reticle, nodeColor=<rgba 0 0 0 1>}
  n4 = NamedRerouteUsage()
  LinearInterpolate(
   A=VectorParameter() {group=0 - Global, parameterName=Outer Color, sortPriority=1, defaultValue=<rgba 0 0 0 1>}.RGB
   B=VectorParameter() {group=0 - Global, parameterName=Color, sortPriority=0, defaultValue=<rgba 1 1 1 1>}.RGB
   Alpha=Saturate(None=Add(A=@n3, B=@n4))
  )
  n5 = NamedRerouteUsage()
  n6 = ScalarParameter() {defaultValue=1, group=2 - Triangles, parameterName=Triangle Opacity, sliderMax=1, sortPriority=0}
  n7 = Saturate(None=Add(A=NamedRerouteUsage(), B=Multiply(A=@n5, B=@n6)))
  n8 = Multiply(
   A=NamedRerouteUsage()
   B=ScalarParameter() {defaultValue=0.5, group=3 - Targeting Circle, parameterName=Targeting Circle Opacity, sliderMax=1, sortPriority=0}
  )
  Multiply(
   A=Saturate(None=Add(A=@n7, B=@n8))
   B=ScalarParameter() {defaultValue=1, group=0 - Global, parameterName=Global Opacity, sliderMax=1, sortPriority=2}
  )
  n9 = ScalarParameter() {defaultValue=0.98, group=2 - Triangles, parameterName=Triangle Feather Start, sortPriority=1}
  n10 = ScalarParameter() {defaultValue=0.6, group=2 - Triangles, parameterName=Triangle Feather End, sortPriority=2}
  n11 = NamedRerouteUsage()
  n12 = MaterialFunctionCall(
   UVs=Add(
    A=MaterialFunctionCall!(UV=@n11, Swizzle UV) {materialFunction=/FortUILibrary/MaterialFunctions/Basics/MF_UI_Symmetry_2x2_New}.Symmetry
    B=ScalarParameter() {defaultValue=0.1, group=2 - Triangles, parameterName=Triangles Offset, sortPriority=3}
   )
   "Rotation Angle (0-1)"=Constant() {r=0.125}
  ) {materialFunction=&CustomRotator}
  n13 = Constant() {r=0.25}
  n14 = AppendVector(A=Constant(), B=ScalarParameter() {defaultValue=0.5, group=2 - Triangles, parameterName=Triangle Size})
  n15 = OneMinus(None=MaterialFunctionCall!(
    UVs=MaterialFunctionCall(UVs=@n12.Rotated Values, Texture Scale=@n13) {materialFunction=&ScaleUVsByCenter}.UVs
    vertex_0
    vertex_1=@n14
    vertex_2=@n14
    vertex_3
    vertex_4
    vertex_5
    Outline Thickness
    Glow Min
    Glow Max
   ) {materialFunction=&/MF_UI_ShockwaveMace_Polygon_6_Vertex_WIP}.SDF)
  NamedRerouteDeclaration(None=SmoothStep(Min=Subtract(A=@n9, B=@n10) {constB=0.05}, Max=@n9, Value=Saturate(None=@n15))) {name=Triangles, nodeColor=<rgba 0 0 0 1>}
  n16 = ScalarParameter() {defaultValue=1, group=1 - Circle, parameterName=Circle Feather Start, sortPriority=4}
  n17 = MaterialFunctionCall(
   UVs=NamedRerouteUsage()
   Size=ScalarParameter() {defaultValue=0.7, group=1 - Circle, parameterName=Circle Size, sliderMax=1, sortPriority=0}
   OutlineThickness=ScalarParameter() {defaultValue=0.01, group=1 - Circle, parameterName=Circle Thickness, sliderMax=1, sortPriority=1}
  ) {materialFunction=&MF_UI_Circle}
  n18 = SmoothStep(
   Min=Subtract(A=@n16, B=ScalarParameter() {defaultValue=0.06, group=1 - Circle, parameterName=Circle Feather End, sortPriority=5}) {constB=0.05}
   Max=@n16
   Value=Saturate(None=OneMinus(None=@n17.SDFOutline))
  )
  n19 = MaterialFunctionCall!(
   UV=NamedRerouteUsage()
   Segment Count=Constant() {r=2}
   Segment Spacing=ScalarParameter() {defaultValue=0.1, group=1 - Circle, parameterName=Segment Spacing, sortPriority=2}
   Fixed Spacing
  ) {materialFunction=/FortUILibrary/MaterialFunctions/Basics/MF_UI_RadialSegments}
  NamedRerouteDeclaration(None=Saturate(None=Subtract(A=@n18, B=@n19.Segmented Triangles Fill))) {name=Circle, nodeColor=<rgba 0 0 0 1>}
  NamedRerouteDeclaration(None=MaterialFunctionCall(
    UVs=MaterialFunctionCall() {materialFunction=&GetUserInterfaceUV}.Normalized UV
    "Rotation Angle (0-1)"=ScalarParameter() {group=0 - Global, parameterName=Global Rotation}
   ) {materialFunction=&CustomRotator}.Rotated Values) {name=UV, nodeColor=<rgba 0 0 0 1>}
  n20 = ScalarParameter() {defaultValue=1, group=3 - Targeting Circle, parameterName=Targeting Circle Feather Start, sortPriority=4}
  n21 = Subtract(A=@n20, B=ScalarParameter() {defaultValue=0.1, group=3 - Targeting Circle, parameterName=Targeting Circle Feather End, sortPriority=5}) {constB=0.05}
  n22 = NamedRerouteUsage()
  n23 = ScalarParameter() {defaultValue=0.25, group=3 - Targeting Circle, parameterName=Targeting Circle Size, sliderMax=1, sortPriority=1}
  n24 = ScalarParameter() {defaultValue=0.001, group=3 - Targeting Circle, parameterName=Targeting Circle Thickness, sliderMax=1, sortPriority=2}
  n25 = Saturate(None=OneMinus(None=MaterialFunctionCall(UVs=@n22, Size=@n23, OutlineThickness=@n24) {materialFunction=&MF_UI_Circle}.SDFOutline))
  NamedRerouteDeclaration(None=Saturate(None=SmoothStep(Min=@n21, Max=@n20, Value=@n25))) {name=Targeting Circle, nodeColor=<rgba 0 0 0 1>}
  ScalarParameter() {defaultValue=0.048, group=1 - Circle, parameterName=Spacing Falloff, sliderMax=1, sortPriority=3}
 ]
}
