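# Weapon reticle extension widget for the Shockwave Mace, showing an animated material reticle and a radial ring of ammo stack entries.
# Stacks are rebuilt on construct and clip attribute changes, while local ammo changes trigger fill or use on each entry.
# Weapon fire plays the slam animation, fading the stacks, until the mace's ground-contact event reverses it.
# An async gameplay-tag wait on the player pawn triggers the bounce animation.
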WidgetBlueprint /ShockwaveMace/UI/WBP_Reticle_ShockwaveMace
animations [
 Anim_Intro {
  binding_count=1
  length_seconds=0.3
  object=&self:Anim_Intro
  sequence {bindings [Reticle {
     id=#1
     possessed_class=Overlay
     tracks [
      {
       type=MovieSceneFloatTrack
       display_name=Render Opacity
       sections [{type=MovieSceneFloatSection, channels [None {type=MovieSceneScriptingFloatChannel, keys [[0,0],[6,1]]}]}]
      }
      {
       type=MovieScene2DTransformTrack
       sections [{
         type=MovieScene2DTransformSection
         channels [
          Translation.X {type=MovieSceneScriptingFloatChannel}
          Translation.Y {type=MovieSceneScriptingFloatChannel}
          Angle {type=MovieSceneScriptingFloatChannel}
          Scale.X {type=MovieSceneScriptingFloatChannel, keys [[0,1.25],[6,1]]}
          Scale.Y {type=MovieSceneScriptingFloatChannel, keys [[0,1.25],[6,1]]}
          Shear.X {type=MovieSceneScriptingFloatChannel}
          Shear.Y {type=MovieSceneScriptingFloatChannel}
         ]
        }]
      }
     ]
    }]}
 }
 Anim_Bounce {
  binding_count=1
  length_seconds=0.4753
  object=&self:Anim_Bounce
  sequence {
   bindings [ReticleVisual {
     id=#2
     possessed_class=Image
     tracks [
      {
       type=MovieSceneWidgetMaterialTrack
       sections [{
         type=MovieSceneParameterSection
         channels [Global Rotation {type=MovieSceneScriptingFloatChannel, keys [[0,0],[5.397,0],[28,2],[28.516,-0.25]]}]
        }]
      }
      {
       type=MovieScene2DTransformTrack
       sections [{
         type=MovieScene2DTransformSection
         end_frame=28
         start_frame=0
         channels [
          Translation.X {type=MovieSceneScriptingFloatChannel, const [0,1]}
          Translation.Y {type=MovieSceneScriptingFloatChannel, const [0,1]}
          Angle {type=MovieSceneScriptingFloatChannel, const [0,1]}
          Scale.X {type=MovieSceneScriptingFloatChannel, keys [[0,1],[28,1.2]]}
          Scale.Y {type=MovieSceneScriptingFloatChannel, keys [[0,1],[28,1.2]]}
          Shear.X {type=MovieSceneScriptingFloatChannel, const [0,1]}
          Shear.Y {type=MovieSceneScriptingFloatChannel, const [0,1]}
         ]
        }]
      }
     ]
    }]
   display_rate {numerator=60}
  }
 }
 Anim_Slam {
  binding_count=2
  length_seconds=0.2515
  object=&self:Anim_Slam
  sequence {
   bindings [
    ReticleVisual {
     id=#3
     possessed_class=Image
     tracks [
      {
       type=MovieSceneWidgetMaterialTrack
       sections [{
         type=MovieSceneParameterSection
         channels [
          Targeting Circle Opacity {type=MovieSceneScriptingFloatChannel, keys [[0,0],[9.052,0.4],[15.086,0.5]]}
          Triangle Size {type=MovieSceneScriptingFloatChannel, keys [[0,1],[9.052,2.5],[15.086,2.5]]}
         ]
        }]
      }
      {
       type=MovieSceneFloatTrack
       display_name=Render Opacity
       sections [{type=MovieSceneFloatSection, channels [None {type=MovieSceneScriptingFloatChannel, keys [[0,1],[15.086,0.4]]}]}]
      }
     ]
    }
    DynamicEntryBox_Stacks {
     id=#4
     possessed_class=DynamicEntryBox
     tracks [{
       type=MovieSceneFloatTrack
       display_name=Render Opacity
       sections [{type=MovieSceneFloatSection, channels [None {type=MovieSceneScriptingFloatChannel, keys [[0,1],[9.052,0],[15.086,0]]}]}]
      }]
    }
   ]
   display_rate {numerator=60}
  }
 }
]
graphs [
 EventGraph {nodes [
   OnGroundContact() {is=title K2Node_CustomEvent}
   > Branch(Condition=Get IsSlamActive() -> (IsSlamActive:bool))
   > Set IsSlamActive(IsSlamActive:bool=false) -> (Output_Get:bool)
   n1 = Get Anim_Slam() -> (Anim_Slam:Widget Animation Object Reference)
   > PlayAnimationReverse(InAnimation=@n1)
   n2 = OnLocalAmmoChanged!() -> (OutputDelegate, then, LocalCount:int, LocalRemaining:int) {is=title K2Node_CustomEvent}
   > n3 = For Each Loop(Array:Array of User Widget Object References=GetAllEntries!(self:Dynamic Entry Box Base Object Reference=Get DynamicEntryBox_Stacks() -> (DynamicEntryBox_Stacks:Dynamic Entry Box Object Reference)) -> (ReturnValue:Array of User Widget Object References)) -> (Array Element:User Widget Object Reference)
   > n4 = Cast To WBP_Reticle_ShockwaveMace_Stack!(execute, Object=@n3) -> (then, CastFailed, AsWBP Reticle Shockwave Mace Stack:WBP Reticle Shockwave Mace Stack Object Reference) {is=title K2Node_DynamicCast}
   > n5 = Branch(Condition="integer < integer"(A=@n3.Array Index, B=@n2.LocalCount))
   > OnFill(self:WBP Reticle Shockwave Mace Stack Object Reference=@n4)
   OnUse(execute=@n5.else, self:WBP Reticle Shockwave Mace Stack Object Reference=@n4)
   WaitForBounce() {is=title K2Node_CustomEvent}
   > n6 = Cast To FortPlayerController(Object=GetOwningPlayer())
   > n7 = Cast To FortPlayerPawnAthena(Object=GetPlayerPawn!(self:Fort Player Controller Object Reference=@n6) -> (ReturnValue:Fort Player Pawn Object Reference))
   > n8 = WaitGameplayTagAddToActor!(
    execute:Exec
    TargetActor:Actor Object Reference=@n7
    Tag:Gameplay Tag Structure=(TagName="Abilities.ShockwaveMace.Bounce")
    OnlyTriggerOnce:bool=false
   ) -> (then:Exec, AsyncTaskProxy:Ability Async Wait Gameplay Tag Added Object Reference, Added:Exec) {is=title K2Node_AsyncAction}
   PlayAnimationForward(execute=@n8.Added, InAnimation=Get Anim_Bounce() -> (Anim_Bounce:Widget Animation Object Reference))
   Event Construct()
   > PlayAnimation(InAnimation=Get Anim_Intro() -> (Anim_Intro:Widget Animation Object Reference))
   n9 = GetEquippedWeapon!(self:Fort Weapon Reticle Extension Widget Base Object Reference) -> (ReturnValue:Fort Weapon Object Reference)
   > GenerateStacks!(
    execute
    self
    TotalStacks:int=GetBulletsPerClip!(self:Fort Weapon Object Reference=@n9) -> (ReturnValue:int=0)
    CurrentStacks:int=GetLocalMagazineAmmoCount!(self:Fort Weapon Object Reference=@n9) -> (ReturnValue:int=0)
   ) -> (then)
   > WaitForBounce() {is=title K2Node_CallFunction}
   > n10 = Set EquippedWeapon(EquippedWeapon:Fort Weapon Object Reference=GetEquippedWeapon()) -> (Output_Get:Fort Weapon Object Reference)
   n11 = Create Event()
   > Assign On Local Ammo Changed!(execute:Exec, self:Fort Weapon Object Reference=@n10, Delegate:"Delegate (by ref)"=@n11) -> (then:Exec) {is=title K2Node_AssignDelegate}
   > n12 = Cast To B_ShockwaveMace!(execute, Object=@n10) -> (then, CastFailed, AsB Shockwave Mace:B Shockwave Mace Object Reference) {is=title K2Node_DynamicCast}
   n13 = Create Event()
   > Bind Event to On Ground Contact(self:B Shockwave Mace Object Reference=@n12, Delegate=@n13) {is=title K2Node_AddDelegate}
   Event OnWeaponSingleFire() {is=title K2Node_Event}
   > n14 = Branch(Condition=Get IsSlamActive())
   Set IsSlamActive(execute=@n14.else, IsSlamActive=true)
   > PlayAnimationForward(InAnimation=@n1)
   Event Destruct()
   n15 = Get EquippedWeapon() -> (EquippedWeapon:Fort Weapon Object Reference)
   > Is Valid(InputObject=@n15)
   > Unbind Event from On Local Ammo Changed(self:Fort Weapon Object Reference=@n15, Delegate=@n11) {is=title K2Node_RemoveDelegate}
   > n16 = Cast To B_ShockwaveMace(Object=@n15)
   > Unbind Event from On Ground Contact(self:B Shockwave Mace Object Reference=@n16, Delegate=@n13) {is=title K2Node_RemoveDelegate}
   Event OnAmmoClipAttributeChanged!() -> (OutputDelegate, then, AmmoClipValue:float) {is=title K2Node_Event}
   n17 = GetEquippedWeapon()
   > GenerateStacks(TotalStacks=GetBulletsPerClip(self=@n17) -> (ReturnValue=0), CurrentStacks=GetLocalMagazineAmmoCount(self=@n17) -> (ReturnValue=0))
  ]}
 GenerateStacks {nodes [
   n1 = GenerateStacks!() -> (then, TotalStacks:int, CurrentStacks:int) {is=title K2Node_FunctionEntry}
   > Reset!(execute, self:Dynamic Entry Box Object Reference=Get DynamicEntryBox_Stacks(), bDeleteWidgets:bool=false) -> (then)
   > n2 = For Loop(FirstIndex=1, LastIndex=@n1)
   > n3 = Create Entry!(execute, self:Dynamic Entry Box Object Reference=Get DynamicEntryBox_Stacks()) -> (then, ReturnValue:User Widget Object Reference)
   > n4 = Cast To WBP_Reticle_ShockwaveMace_Stack(Object=@n3)
   > SetFill!(execute, self:WBP Reticle Shockwave Mace Stack Object Reference=@n4, IsFilled:bool="integer <= integer"(A=@n2, B=@n1.CurrentStacks)) -> (then)
  ]}
]
variables [
 EquippedWeapon {type=FortWeapon, default=null}
 IsSlamActive {type=boolean}
]
widgets {
 info {
  inheritedWidgetCount=1
  parentClass=/Script/FortniteUI.FortWeaponReticleExtensionWidgetBase
  rootWidgetClass=&Overlay
  widgetCount=6
 }
 widgets [
  Reticle {widgetClassPath=&Overlay, widget=Reticle}
  SizeBox_Reticle {
   widgetClassPath=&SizeBox
   parent=Reticle
   widget=SizeBox_Reticle
   properties {bOverride_HeightOverride=true, bOverride_WidthOverride=true, heightOverride=96, widthOverride=96}
   slot {class=OverlaySlot, properties {horizontalAlignment=HAlign_Center, verticalAlignment=VAlign_Center}}
  }
  ReticleVisual {
   widgetClassPath=&Image
   bIsVariable=true
   parent=SizeBox_Reticle
   widget=ReticleVisual
   properties {
    brush {
     bIsDynamicallyLoaded=false
     drawAs=Image
     imageType=NoImage
     mirroring=NoMirror
     resourceName=None
     resourceObject=&/MI_UI_ShockwaveMaceReticle
     tiling=NoTile
     imageSize=<32 32>
     margin=<ltrb 0 0 0 0>
     outlineSettings {
      bUseBrushTransparency=false
      roundingType=HalfHeightRadius
      width=0
      color {colorUseRule=UseColor_Specified, specifiedColor=<rgba 0 0 0 0>}
      cornerRadii=<0 0 0 0>
     }
     tintColor {colorUseRule=UseColor_Specified, specifiedColor=<rgba 1 1 1 1>}
     uVRegion {bIsValid=false, max=<0 0>, min=<0 0>}
    }
    renderTransform {angle=0, scale=<0.8 0.8>, shear=<0 0>, translation=<0 0>}
   }
   slot {class=SizeBoxSlot}
  }
  SizeBox_0 {
   widgetClassPath=&SizeBox
   parent=Reticle
   widget=SizeBox_0
   properties {bOverride_HeightOverride=true, bOverride_WidthOverride=true, heightOverride=192, widthOverride=192}
   slot {class=OverlaySlot, properties {horizontalAlignment=HAlign_Fill, verticalAlignment=VAlign_Fill}}
  }
  DynamicEntryBox_Stacks {
   widgetClassPath=&DynamicEntryBox
   bIsVariable=true
   parent=SizeBox_0
   widget=DynamicEntryBox_Stacks
   properties {
    entryBoxType=Radial
    entryWidgetClass=&/WBP_Reticle_ShockwaveMace_Stack.WBP_Reticle_ShockwaveMace_Stack_C
    numDesignerPreviewEntries=4
    radialBoxSettings {angleBetweenItems=25, bClockwiseOrder=false, bDistributeItemsEvenly=false, marginSize=0, sectorCentralAngle=360, startingAngle=270}
   }
   slot {class=SizeBoxSlot}
  }
  TutorialHighlight {widgetClassPath=&TutorialHighlight}
 ]
}
